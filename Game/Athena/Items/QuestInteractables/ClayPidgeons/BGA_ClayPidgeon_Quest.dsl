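; Clay pigeon quest target actor that occasionally launches as a duck instead of a disk.
; BeginPlay rolls the duck chance on the server, reveals the matching mesh, plays launch sounds and binds the damaged and died events.
; Those events set ShotFX and HitWorldFX flags, with OnRep functions reacting on remote machines only.
; The shot OnRep spawns duck or clay destruction effects once, while the world-hit OnRep only prints debug text.
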(event EventBeginPlay
  (bind _sm_duck (Variables|Default|GetSM_Duck))
  (Networking|SwitchHasAuthority
    (:Authority
      ; Only Roll for Duck on the server
      (Variables|Default|SetIsDuck (> (Math|Random|RandomIntegerinRange 0 100) 90)))
    (:Remote
      _))
  (if (Variables|Default|GetIsDuck)
    (Development|SetHiddeninGame _sm_duck false false)
    (Audio|SpawnSoundAttached (Variables|Default|GetDuckLaunchSound) _sm_duck)
    (else
      (Development|SetHiddeninGame (Variables|Default|GetSM_Trap_Disk) false false)))
  (Audio|PlaySoundatLocation (Variables|Default|GetLauncherFireSound) (Transformation|GetActorLocation)) ; Play Shoot Sound
  ;[1 1 1] HandleFX
  (EventDispatchers|BindEventtoOnDied (delegate Custom|HeDed))
  (EventDispatchers|BindEventtoOnDamaged (delegate Custom|HeDamaged)))

(event Collision|EventActorBeginOverlap
  (params (OtherActor :type Actor%20Object%20Reference)))

(event Custom|HeDed
  (params (DamagedActor :type Actor%20Object%20Reference) (Damage :type Float%20%28single-precision%29) (InstigatedBy :type Controller%20Object%20Reference) (DamageCauser :type Actor%20Object%20Reference) (HitLocation :type Vector) (FHitComponent :type Primitive%20Component%20Object%20Reference) (BoneName :type Name) (Momentum :type Vector))
  ;[1 1 1] HandleFX
  (Variables|Default|SetHitWorldFX true))

(event Custom|HeDamaged
  (params (DamagedActor :type Actor%20Object%20Reference) (Damage :type Float%20%28single-precision%29) (InstigatedBy :type Controller%20Object%20Reference) (DamageCauser :type Actor%20Object%20Reference) (HitLocation :type Vector) (FHitComponent :type Primitive%20Component%20Object%20Reference) (BoneName :type Name) (Momentum :type Vector))
  ;[1 1 1] HandleFX
  (Variables|Default|SetShotFX true))

(event EventTick
  (params (DeltaSeconds :type Float%20%28single-precision%29)))

(fn OnRep_HeDedFX
  (Networking|SwitchHasAuthority
    (:Remote
      (Development|PrintString "HIT WORLD" false false "(R=1.000000,G=0.109999,B=0.000000,A=1.000000)"))))

(fn OnRep_ShotFX
  (Networking|SwitchHasAuthority
    (:Remote
      (Utilities|FlowControl|DoOnce false
        (:Completed
          (Development|PrintString "SHOT!!" false false "(R=0.336960,G=1.000000,B=0.000000,A=1.000000)")
          (if (Variables|Default|GetIsDuck)
            (Effects|SpawnEmitteratLocation (Variables|Default|GetDuckDestroyFX) (Transformation|GetActorLocation) (Transformation|GetActorRotation))
            (Audio|PlaySoundatLocation (Variables|Default|GetDuckHitSound) (Transformation|GetActorLocation)) ; Play Shoot Sound
            (else
              (Effects|SpawnEmitteratLocation (Variables|Default|GetClayPidgeonDestroyedFX) (Transformation|GetActorLocation) (Transformation|GetActorRotation)))))))))

(fn ConstructionScript)
