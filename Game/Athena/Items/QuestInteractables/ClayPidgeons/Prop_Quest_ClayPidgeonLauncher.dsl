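; Interactable clay pigeon launcher quest prop that runs a shooting score challenge.
; Interacting starts a sling loop that spawns pigeons from the server, rising from one to three per volley as the count grows.
; Shot pigeons add score and record contributors; one hitting the world stops the run, updates the high score and, outside Creative, credits two quests at goal score.
; RepNotifies drive score-sign materials, effects and client arm timelines, resynced on firing.
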(event EventBeginPlay
  (bind _returnvalue (Fortnite|GetLocalFortPlayerControllers true))
  (bind _output (Utilities|Array|Get(acopy) _returnvalue))
  ;[1 0.5722 0] HACKY SACK
  (Utilities|FlowControl|Delay 5.0)
  ;[1 1 1] Check if In Creative Mode or not
  (bind _asfort_game_state_athena (Utilities|Casting|CastToFortGameStateAthena (Game|GetGameState)))
  (if (Variables|/Script/FortniteGame.FortGameStateAthena|GetbPlaylistDataIsLoaded _asfort_game_state_athena)
    (bind _cancompletequest (Class|AthenaQuestFunctionLibrary|AthenaCheckCreativeMode _asfort_game_state_athena))
    (if _cancompletequest
      (Variables|Default|SetCreativeMode false)
      (else
        (Variables|Default|SetCreativeMode true)))
    ;[1 1 1] Init Full Stack Pidgeon Loader Loc
    (Networking|SwitchHasAuthority
      (:Remote
        (Variables|Default|SetFullStackLoc (Variables|/Script/Engine.SceneComponent|GetRelativeLocation (Variables|Default|GetSM_Trap_Disk_Stack)))
        ;[1 1 1] Quest Bindings
        (bind (_questisvalid _questcompleted _objectivecompleted) (Default|AthenaCheckQuestandObjectiveCompletion _output (Variables|QuestStuff|GetQuestItem) (Variables|QuestStuff|GetObjBackendName)))
        (if _questisvalid
          (if _objectivecompleted
            _
            (elif (Variables|Default|GetCreativeMode)
              _
              (else
                (Quests|BindEventtoOnQuestsUpdated (Quests|GetQuestManager _output "Invalid") (EventDispatchers|CreateEvent :!event Custom|QuestCompleted))
                (Variables|QuestStuff|SetQuestLocallyActive true)))))))
    (else
      (Playlist|BindEventtoOnCurrentPlaylistLoaded _asfort_game_state_athena (delegate Custom|PlaylistLoaded)))))

(event Custom|PlaylistLoaded
  (params (PlaylistName :type Name) (PlaylistContextTags :type Gameplay%20Tag%20Container%20Structure%20%28by%20ref%29))
  (bind _returnvalue (Fortnite|GetLocalFortPlayerControllers true))
  (bind _output (Utilities|Array|Get(acopy) _returnvalue))
  ;[1 1 1] Check if In Creative Mode or not
  (bind _cancompletequest (Class|AthenaQuestFunctionLibrary|AthenaCheckCreativeMode (Game|GetGameState)))
  (if _cancompletequest
    (Variables|Default|SetCreativeMode false)
    (else
      (Variables|Default|SetCreativeMode true)))
  ;[1 1 1] Init Full Stack Pidgeon Loader Loc
  (Networking|SwitchHasAuthority
    (:Remote
      (Variables|Default|SetFullStackLoc (Variables|/Script/Engine.SceneComponent|GetRelativeLocation (Variables|Default|GetSM_Trap_Disk_Stack)))
      ;[1 1 1] Quest Bindings
      (bind (_questisvalid _questcompleted _objectivecompleted) (Default|AthenaCheckQuestandObjectiveCompletion _output (Variables|QuestStuff|GetQuestItem) (Variables|QuestStuff|GetObjBackendName)))
      (if _questisvalid
        (if _objectivecompleted
          _
          (elif (Variables|Default|GetCreativeMode)
            _
            (else
              (Quests|BindEventtoOnQuestsUpdated (Quests|GetQuestManager _output "Invalid") (EventDispatchers|CreateEvent :!event Custom|QuestCompleted))
              (Variables|QuestStuff|SetQuestLocallyActive true))))))))

(event Custom|QuestCompleted
  ;[1 1 1] Quest Updated
  (if (Variables|QuestStuff|GetQuestLocallyActive)
    (bind (_questisvalid _questcompleted _objectivecompleted) (Default|AthenaCheckQuestandObjectiveCompletion (Utilities|Array|Get(acopy) (Fortnite|GetLocalFortPlayerControllers true)) (Variables|QuestStuff|GetQuestItem) (Variables|QuestStuff|GetObjBackendName)))
    (if (or _questcompleted _objectivecompleted)
      (Utilities|FlowControl|Delay 0.2)
      (Variables|QuestStuff|SetQuestLocallyActive false)
      (CallFunction|QuestCompletedFX)
      ;[1 1 1] Quest Bindings
      (Quests|UnbindEventfromOnQuestsUpdated (Quests|GetQuestManager (Utilities|Array|Get(acopy) (Fortnite|GetLocalFortPlayerControllers true)) "Invalid") (EventDispatchers|CreateEvent :!event Custom|QuestCompleted)))))

(event Custom|QuestCompletedFX
  (bind _returnvalue (Fortnite|GetLocalFortPlayerControllers true))
  (bind _output (Utilities|Array|Get(acopy) _returnvalue))
  (bind _returnvalue_2 (Game|Player|GetPlayerPawn _output))
  ;[1 1 1] Quest Success FX
  (Effects|SpawnFXSystemAttached (Variables|Default|GetQuestSuccessEmitter) (Variables|/Script/Engine.Character|GetMesh _returnvalue_2))
  (Audio|PlaySoundatLocation (Variables|Default|GetQuestSuccessSound) (Transformation|GetActorLocation _returnvalue_2)) ; Quest Success Sound
)

(event Interaction|EventBlueprintOnInteract
  (params (InteractingPawn :type Fort%20Pawn%20Object%20Reference) (InteractionBeingAttempted :type EInteractionBeingAttempted%20Enum :default "FirstInteraction"))
  ;[1 1 1] INTERACT
  (Variables|Default|SetInteractingPawn InteractingPawn)
  (Variables|ScoringInfo|SetCurrentScore 0)
  (Variables|Default|SetEngineRunning true))

(event Custom|SlingArmMotion
  (bind _servergametimeatslingstart (Variables|Default|GetServerGameTimeAtSlingStart))
  (bind _clientgametimeatslingstart (Variables|Default|GetClientGameTimeAtSlingStart))
  (bind _returnvalue (- _clientgametimeatslingstart _servergametimeatslingstart))
  (bind _returnvalue_2 (Game|GetGameState))
  (bind _returnvalue_3 (GameState|GetServerWorldTimeSeconds _returnvalue_2))
  (bind _sm_trap_disk_stack (Variables|Default|GetSM_Trap_Disk_Stack))
  (bind _currentpigeonnum (Variables|Default|GetCurrentPigeonNum))
  ;[1 1 1] Firing Movement and FX
  (Utilities|FlowControl|Sequence
    (:then_0
      (Math|Integer|IncrementInt (Variables|Default|GetCurrentPigeonNum)
        (:%20%20
          (Networking|SwitchHasAuthority
            (:Authority
              ;[1 1 1] SERVER - Fire Proj
              (Variables|Default|SetServerGameTimeAtSlingStart _returnvalue_3)
              (Utilities|FlowControl|Delay 2.45)
              (if (not (Variables|Default|GetShuttingDown))
                (Variables|Default|SetFIRE true)
                ;[1 1 1] Decide how many to fire at once
                (if (< _currentpigeonnum (Variables|Default|Get1PigeonThreshold))
                  ;[1 1 1] Fire 1 Pigeon
                  (CallFunction|FirePidgeon (Variables|Default|GetArrowCenterTrans))
                  (Utilities|FlowControl|Delay (Variables|Default|GetDelayBeforeRefire))
                  (CallFunction|SlingArmMotion)
                  (elif (< _currentpigeonnum (Variables|Default|Get2PigeonThreshold))
                    ;[1 1 1] Fire 2 Pigeons
                    (CallFunction|FirePidgeon (Variables|Default|GetArrowRightTrans))
                    (Utilities|FlowControl|Delay 0.2)
                    (CallFunction|FirePidgeon (Variables|Default|GetArrowLeftTrans))
                    (Utilities|FlowControl|Delay (Variables|Default|GetDelayBeforeRefire))
                    (CallFunction|SlingArmMotion)
                    (else
                      ;[1 1 1] Fire 3 Pigeons
                      (CallFunction|FirePidgeon (Variables|Default|GetArrowRightTrans))
                      (Utilities|FlowControl|Delay 0.2)
                      (CallFunction|FirePidgeon (Variables|Default|GetArrowCenterTrans))
                      (Utilities|FlowControl|Delay 0.2)
                      (CallFunction|FirePidgeon (Variables|Default|GetArrowLeftTrans))
                      (Utilities|FlowControl|Delay (Variables|Default|GetDelayBeforeRefire))
                      (CallFunction|SlingArmMotion))))))
            (:Remote
              ;[1 1 1] CLIENT Fire Timeline anim
              (Variables|Default|SetClientGameTimeAtSlingStart _returnvalue_3)
              (Development|PrintString (Utilities|String|ToString(Float) _returnvalue) false false)
              (Variables|Default|SetNewStartTimeClientSling _returnvalue)
              (Components|Timeline|SetPlayRate (Variables|Prop_Quest_ClayPidgeonLauncher|GetSlingLoadTimeline) 1.0)
              (bind (_direction _rotatetime _rocktime) (Timeline|SlingLoadTimeline PlayFromStart)
                (:Update
                  (Transformation|SetRelativeRotation (Variables|Default|GetSM_Trap_Sling) (Math|Rotator|MakeRotator 0.0 0.0 (* _rotatetime -359.0)) false false))))))))
    (:then_1
      (Networking|SwitchHasAuthority
        (:Remote
          (if (<= (.z (Variables|/Script/Engine.SceneComponent|GetRelativeLocation _sm_trap_disk_stack)) -35.0)
            (Transformation|SetRelativeLocation _sm_trap_disk_stack "0, 0,26.000000" false true)))))))

(event Custom|FIRE%20on%20time
  ;[1 1 1] Force Client Firing Motion to happen closer to when Server actually launches the projectile
  (Networking|SwitchHasAuthority
    (:Remote
      (Utilities|FlowControl|Sequence
        (:then_0
          ;[1 1 1] CLIENT Fire Timeline anim
          (Timeline|SlingLoadTimeline Stop))
        (:then_1
          (bind (_direction _rotatetime _rocktime) (Timeline|Timeline_0 PlayFromStart)
            (:Update
              (Transformation|SetRelativeRotation (Variables|Default|GetSM_Trap_Sling) (Math|Rotator|MakeRotator 0.0 0.0 (Math|Float|Lerp -1.8e+02 -359.0 _rotatetime)) false false)
              (Transformation|SetRelativeRotation (Variables|Default|GetSM_MiddleRail) (Math|Rotator|MakeRotator 3e+01 (* _rocktime 2e+01)) false false))))))))

(event Custom|ShutdownEngine
  ;[1 1 1] Reset Arm
  (Utilities|FlowControl|Sequence
    (:then_0
      ;[1 1 1] Firing Movement and FX
      ;[1 1 1] CLIENT Fire Timeline anim
      (Timeline|SlingLoadTimeline Stop))
    (:then_1
      (Variables|Default|SetShuttingDown true)
      (Variables|Default|SetCurrentZRotation (.yaw (Variables|/Script/Engine.SceneComponent|GetRelativeRotation (Variables|Default|GetSM_Trap_Sling))))
      (bind (_direction _resettime) (Timeline|ResetMotion PlayFromStart)
        (:Update
          (Transformation|SetRelativeRotation (Variables|Default|GetSM_Trap_Sling) (Math|Rotator|MakeRotator 0.0 0.0 (Math|Float|Lerp (Variables|Default|GetCurrentZRotation) 0.0 _resettime)) false false))))))

(event Custom|NewPidgeonFired
  (params (PidgeonFired :type Building%20Gameplay%20Actor%20Object%20Reference))
  ;[1 1 1] Bind Destruction of Current Pidgeon
  (Networking|SwitchHasAuthority
    (:Authority
      (EventDispatchers|BindEventtoOnDamaged PidgeonFired (EventDispatchers|CreateEvent :!event Custom|PidgeonShot))
      (Collision|BindEventtoOnActorHit PidgeonFired (EventDispatchers|CreateEvent :!event Custom|PidgeonHitWorld)))))

(event Custom|PidgeonDead
  (params (PidgeonDead :type Building%20Gameplay%20Actor%20Object%20Reference))
  ;[1 1 1] Bind Destruction of Current Pidgeon
  (EventDispatchers|UnbindEventfromOnDamaged PidgeonDead (EventDispatchers|CreateEvent :!event Custom|PidgeonShot))
  (Collision|UnbindEventfromOnActorHit PidgeonDead (EventDispatchers|CreateEvent :!event Custom|PidgeonHitWorld)))

(event Custom|PidgeonShot
  (params (DamagedActor :type Actor%20Object%20Reference) (Damage :type Float%20%28single-precision%29) (InstigatedBy :type Controller%20Object%20Reference) (DamageCauser :type Actor%20Object%20Reference) (HitLocation :type Vector) (FHitComponent :type Primitive%20Component%20Object%20Reference) (BoneName :type Name) (Momentum :type Vector))
  ;[1 1 1] Player Shot Pidgeon - KEEP FIRING
  (bind _asfort_player_controller_athena (Utilities|Casting|CastToFortPlayerControllerAthena (Game|GetInstigatorController DamageCauser)))
  (Utilities|Array|AddUnique (Variables|Default|GetContributingPlayers) _asfort_player_controller_athena)
  (bind _asbga_clay_pidgeon_quest (Utilities|Casting|CastToBGA_ClayPidgeon_Quest DamagedActor))
  (Gameplay|SilentDie _asbga_clay_pidgeon_quest false)
  (Collision|SetActorEnableCollision _asbga_clay_pidgeon_quest false)
  (CallFunction|PidgeonDead _asbga_clay_pidgeon_quest)
  (Variables|ScoringInfo|SetCurrentScore (+ (Variables|ScoringInfo|GetCurrentScore) 1)))

(event Custom|PidgeonHitWorld
  (params (SelfActor :type Actor%20Object%20Reference) (OtherActor :type Actor%20Object%20Reference) (NormalImpulse :type Vector) (Hit :type Hit%20Result%20Structure%20%28by%20ref%29))
  (bind _currentscore (Variables|ScoringInfo|GetCurrentScore))
  (bind _questitem2 (Variables|QuestStuff|GetQuestItem2))
  (bind _objbackendname (Variables|QuestStuff|GetObjBackendName))
  (bind _questitem (Variables|QuestStuff|GetQuestItem))
  (bind _objbackendname2 (Variables|QuestStuff|GetObjBackendName2))
  ;[1 1 1] Pidgeon Hit World - SHUTDOWN
  (Development|PrintString "HIT WORLD" false false)
  (bind _asbga_clay_pidgeon_quest (Utilities|Casting|CastToBGA_ClayPidgeon_Quest SelfActor))
  (Variables|Default|SetHitWorldLocation (Transformation|GetActorLocation _asbga_clay_pidgeon_quest))
  (Gameplay|SilentDie _asbga_clay_pidgeon_quest false)
  (CallFunction|PidgeonDead _asbga_clay_pidgeon_quest)
  (CallFunction|ShutdownEngine)
  (if (> _currentscore (Variables|ScoringInfo|GetHighScore))
    (Variables|ScoringInfo|SetHighScore _currentscore))
  (Variables|Default|SetEngineRunning false)
  (if (Variables|Default|GetCreativeMode)
    _
    (elif (>= (Variables|ScoringInfo|GetCurrentScore) (Variables|Default|GetGoalScore))
      ;[1 1 1] Stuff for checking both quests without firing twice
      (for _array_element (Variables|Default|GetContributingPlayers)
        (Utilities|FlowControl|Sequence
          (:then_0
            (Utilities|IsValid _array_element
              (:Is%20Valid
                (Utilities|IsValid (Quests|GetQuestwithDefinition (Quests|GetQuestManager _array_element "Invalid") _questitem)
                  (:Is%20Valid
                    (if (not (Quest|HasCompletedObjectivewithName (Quests|GetQuestwithDefinition (Quests|GetQuestManager _array_element "Invalid") _questitem) _objbackendname :!class /Script/FortniteGame.FortQuestItem))
                      (Quest|SendCustomStatEventDirect (Quests|GetQuestManager _array_element "Invalid") _objbackendname (Quests|GetQuestwithDefinition (Quests|GetQuestManager _array_element "Invalid") _questitem) 1)))))))
          (:then_1
            (Utilities|IsValid _array_element
              (:Is%20Valid
                (Utilities|IsValid (Quests|GetQuestwithDefinition (Quests|GetQuestManager _array_element "Invalid") _questitem2)
                  (:Is%20Valid
                    (if (not (Quest|HasCompletedObjectivewithName (Quests|GetQuestwithDefinition (Quests|GetQuestManager _array_element "Invalid") _questitem2) _objbackendname2 :!class /Script/FortniteGame.FortQuestItem))
                      (Quest|SendCustomStatEventDirect (Quests|GetQuestManager _array_element "Invalid") _objbackendname2 (Quests|GetQuestwithDefinition (Quests|GetQuestManager _array_element "Invalid") _questitem2) 1)))))))
          (:then_2
            (Variables|ScoringInfo|SetCurrentScore 0)
            (Utilities|Array|Clear (Variables|Default|GetContributingPlayers)))))))
  (Utilities|FlowControl|Delay 0.5)
  (Variables|ScoringInfo|SetCurrentScore 0)
  (Utilities|Array|Clear (Variables|Default|GetContributingPlayers)))

(event Custom|WinFX
  ;[1 1 1] Win FX
  (Rendering|Material|SetScalarParameterValue (Variables|ScoringInfo|GetMID_HighScore) "Win" 1.0 :!class /Script/Engine.MaterialInstanceDynamic)
  (Utilities|FlowControl|Delay 0.5)
  (Audio|PlaySoundatLocation (Variables|Default|GetNewHighScoreSound) (Transformation|GetActorLocation)) ; High Score Sound
  (Components|Activation|Activate (Variables|Default|GetP_Athena_SoccerGame_ScoredGoal) false)
  (Utilities|FlowControl|Delay 1.5)
  (Rendering|Material|SetScalarParameterValue (Variables|ScoringInfo|GetMID_HighScore) "Win" 0.0 :!class /Script/Engine.MaterialInstanceDynamic))

(event Custom|Score%20Up
  ;[1 1 1] Score Up Sound
  (Networking|SwitchHasAuthority
    (:Remote
      (Utilities|FlowControl|Delay 0.75)
      (Audio|PlaySoundatLocation (Variables|Default|GetScore%20Up%20Sound) (Transformation|GetWorldLocation (Variables|Default|Gethackyscoresign))) ; Score Up Sound
    )))

(event Custom|ScoreReset%20Sound
  ;[1 1 1] Reset Score sound
  (Networking|SwitchHasAuthority
    (:Remote
      (Audio|PlaySoundatLocation (Variables|Default|GetResetScoreSound) (Transformation|GetWorldLocation (Variables|Default|Gethackyscoresign))) ; Score Up Sound
    )))

(fn BlueprintCanInteract
  (params (InteractingPawn :type Fort%20Pawn%20Object%20Reference) (InteractionBeingAttempted :type EInteractionBeingAttempted%20Enum :default "FirstInteraction") (InteractionType :type TInteractionType%20Enum :default "IT_NoInteraction"))
  (return (not (Variables|Default|GetEngineRunning))))

(fn BlueprintGetInteractionString
  (params (InteractingPawn :type Fort%20Pawn%20Object%20Reference) (InteractionBeingAttempted :type EInteractionBeingAttempted%20Enum :default "FirstInteraction"))
  (return (Variables|Default|GetInteractionString)))

(fn FirePidgeon
  (params (ArrowTransform :type Transform))
  (bind _returnvalue (Building|SpawnBuildingGameplayActor "/Game/Athena/Items/QuestInteractables/ClayPidgeons/BGA_ClayPidgeon_Quest.BGA_ClayPidgeon_Quest_C" ArrowTransform (Variables|Default|GetInteractingPawn)))
  (CallFunction|NewPidgeonFired _returnvalue))

(fn OnRep_CurrentScore
  (bind _currentscore (Variables|ScoringInfo|GetCurrentScore))
  (Rendering|Material|SetScalarParameterValue (Variables|ScoringInfo|GetMID_CurrentScore) "Score" (Math|Conversions|ToFloat(Integer) _currentscore) :!class /Script/Engine.MaterialInstanceDynamic)
  (if (!= _currentscore 0)
    (CallFunction|ScoreUp)
    (else
      (CallFunction|ScoreResetSound))))

(fn OnRep_EngineRunning
  (bind _claypidgeon_enginesound_cue (Variables|Default|GetClayPidgeon_EngineSound_Cue))
  (if (Variables|Default|GetEngineRunning)
    (Components|Activation|Activate _claypidgeon_enginesound_cue true)
    (Components|Timeline|SetPlayRate (Variables|Prop_Quest_ClayPidgeonLauncher|GetSlingLoadTimeline) 1.5)
    (Variables|Default|SetCurrentPigeonNum 0)
    (Variables|Default|SetSlingArmSpeed 1.0)
    (Variables|Default|SetShuttingDown false)
    (CallFunction|SlingArmMotion)
    (else
      (Components|Activation|Deactivate _claypidgeon_enginesound_cue)
      (Audio|PlaySoundatLocation (Variables|SoundsAndFX|GetShutOffSound) (Transformation|GetActorLocation)) ; ShutOff sound
      (Variables|Default|SetShuttingDown true))))

(fn OnRep_FIRE
  (CallFunction|FIREOnTime))

(fn OnRep_HighScore
  (Rendering|Material|SetScalarParameterValue (Variables|ScoringInfo|GetMID_HighScore) "Score" (Math|Conversions|ToFloat(Integer) (Variables|ScoringInfo|GetHighScore)) :!class /Script/Engine.MaterialInstanceDynamic)
  (CallFunction|WinFX))

(fn OnRep_HitWorldLocation
  (Effects|SpawnEmitteratLocation (Variables|Default|GetHitWorldEmitter) (Variables|Default|GetHitWorldLocation)))

(fn ConstructionScript
  (Utilities|FlowControl|Sequence
    (:then_0
      (Transformation|SetRelativeLocationAndRotation (Variables|Default|Gethackyscoresign) (Variables|ScoringInfo|GetScoreSignOffset) (Variables|ScoringInfo|GetScoreSignRotation) false false)
      (bind _returnvalue (Rendering|Material|CreateDynamicMaterialInstance (Variables|Default|Gethackyscoresign) 1 0 "None" :!class /Script/Engine.PrimitiveComponent))
      (Variables|ScoringInfo|SetMID_CurrentScore _returnvalue)
      (bind _returnvalue_2 (Rendering|Material|CreateDynamicMaterialInstance (Variables|Default|Gethackyscoresign) 2 0 "None" :!class /Script/Engine.PrimitiveComponent))
      (Variables|ScoringInfo|SetMID_HighScore _returnvalue_2)
      (Rendering|Material|SetScalarParameterValue (Variables|ScoringInfo|GetMID_CurrentScore) "Score" 0.0 :!class /Script/Engine.MaterialInstanceDynamic)
      (Rendering|Material|SetScalarParameterValue (Variables|ScoringInfo|GetMID_HighScore) "Score" 0.0 :!class /Script/Engine.MaterialInstanceDynamic)
      (Rendering|Material|SetVectorParameterValue (Variables|ScoringInfo|GetMID_HighScore) "Color" "(R=1.000000,G=0.797455,B=0.049112,A=1.000000)" :!class /Script/Engine.MaterialInstanceDynamic))
    (:then_1
      (Variables|Default|SetArrowCenterTrans (Math|Transform|MakeTransform (Transformation|GetWorldLocation (Variables|Default|GetArrow)) (Transformation|GetWorldRotation (Variables|Default|GetArrow))))
      (Variables|Default|SetArrowRightTrans (Math|Transform|MakeTransform (Transformation|GetWorldLocation (Variables|Default|GetArrowRight)) (Transformation|GetWorldRotation (Variables|Default|GetArrowRight))))
      (Variables|Default|SetArrowLeftTrans (Math|Transform|MakeTransform (Transformation|GetWorldLocation (Variables|Default|GetArrowLeft)) (Transformation|GetWorldRotation (Variables|Default|GetArrowLeft)))))))
